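; Parent Blueprint for the Transit mobile trailer building prop, setting its exterior look and interior material from enum variables at construction.
; The construction script calls its parent, then maps the tint enum to an index into the standard or alternate texture-data array chosen by exterior design.
; That texture data fills building layer 0, optionally copied to layer 1, and one of three interior material instances goes on mesh slot 1.
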(event EventBeginPlay)

(event Collision|EventActorBeginOverlap
  (params (OtherActor :type Actor%20Object%20Reference)))

(event EventTick
  (params (DeltaSeconds :type Float%20%28single-precision%29)))

(fn SetExteriorPattern
  (Utilities|FlowControl|Sequence
    (:then_0
      (Utilities|Array|Clear (LocalVariables|GetL_AR_TD))
      (switch Utilities|FlowControl|Switch|SwitchonENUM_MobileTrailer_Exterior (Variables|MobileTrailer|GetExteriorDesign)
        (:NewEnumerator0
          (for _array_element (Variables|MobileTrailer|GetTD_StandardStyle)
            (Utilities|Array|Add (LocalVariables|GetL_AR_TD) _array_element)))
        (:NewEnumerator1
          (for _array_element_2 (Variables|MobileTrailer|GetTD_AltStyle)
            (Utilities|Array|Add (LocalVariables|GetL_AR_TD) _array_element_2)))))
    (:then_1
      (bind _output_get (LocalVariables|SetL_ExteriorTextureData (Utilities|Array|Get(acopy) (LocalVariables|GetL_AR_TD) (Variables|MobileTrailer|GetAR_Index))))
      (Utilities|IsValid _output_get
        (:Is%20Valid
          (CallFunction|SetBuildingTextureData 0 (LocalVariables|GetL_ExteriorTextureData)))))
    (:then_2
      (if (Variables|MobileTrailer|GetAdd_TD_to_Layer_2)
        (bind _outtexturedataarray (Building|GetBuildingTextureData))
        (Utilities|IsValid (Utilities|Array|Get(acopy) _outtexturedataarray)
          (:Is%20Valid
            (CallFunction|SetBuildingTextureData 1 (Utilities|Array|Get(acopy) _outtexturedataarray))))))))

(fn SetInteriorMaterial
  (bind _staticmeshcomponent (Variables|Debug|GetStaticMeshComponent))
  (bind _l_matinterior (LocalVariables|GetL_MatInterior))
  (Utilities|FlowControl|Sequence
    (:then_0
      (switch Utilities|FlowControl|Switch|SwitchonENUM_MobileTrailer_Interior (Variables|MobileTrailer|GetInteriorMaterial)
        (:NewEnumerator0
          (LocalVariables|SetL_MatInterior "/Hera/Props/Transit/Transit_MobileTrailer_A/Materials/MI_Transit_MobileTrailerInterior_A.MI_Transit_MobileTrailerInterior_A"))
        (:NewEnumerator1
          (LocalVariables|SetL_MatInterior "/Hera/Props/Transit/Transit_MobileTrailer_A/Materials/MI_Transit_MobileTrailerInterior_B.MI_Transit_MobileTrailerInterior_B"))
        (:NewEnumerator2
          (LocalVariables|SetL_MatInterior "/Hera/Props/Transit/Transit_MobileTrailer_A/Materials/MI_Transit_MobileTrailerInterior_C.MI_Transit_MobileTrailerInterior_C"))))
    (:then_1
      (Utilities|IsValid _staticmeshcomponent
        (:Is%20Valid
          (Utilities|IsValid _l_matinterior
            (:Is%20Valid
              (Rendering|Material|SetMaterial _staticmeshcomponent 1 _l_matinterior))))))))

(fn SetTintColor
  (switch Utilities|FlowControl|Switch|SwitchonENUM_MobileTrailer_Color (Variables|MobileTrailer|GetTintColor)
    (:NewEnumerator0
      (Variables|MobileTrailer|SetAR_Index 0))
    (:NewEnumerator1
      (Variables|MobileTrailer|SetAR_Index 1))
    (:NewEnumerator2
      (Variables|MobileTrailer|SetAR_Index 2))
    (:NewEnumerator3
      (Variables|MobileTrailer|SetAR_Index 3))))

(fn ConstructionScript
  (|CallParentFunction)
  (CallFunction|SetTintColor)
  (CallFunction|SetExteriorPattern)
  (CallFunction|SetInteriorMaterial))
